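# Looping ambient sound cue for the Brimstone decorative pillar, built as close and medium attenuation layers mixed at the root, the medium one quieter.
# Both layers share tonal and low-end loops plus randomly delayed glass hand impacts, rubs and tormented soul bursts.
# The close layer adds whispers and a second impact stream, under two nested looping volume envelopes for irregular swells.
# Looping envelopes drift the tonal loops' pitch and volume, while modulators vary the one-shots except whispers.
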SoundCue /Game/Sounds/Helios/Environment_Actors/BrimestoneDecoPillar/Spot_Brimstone_Deco_Pillar_Loop_Cue
root=Mixer_2
nodes [
 {
  id=Attenuation_1
  children [Enveloper_7]
  properties {attenuationSettings=&/Spot_Brimstone_Pillar_Box_Close_Att}
 }
 {
  id=Attenuation_5
  children [Mixer_0]
  properties {attenuationSettings=&/Spot_Brimstone_Pillar_Box_Medium_Att}
 }
 {id=Delay_0, children [Random_7], properties {delayMax=20, delayMin=6}}
 {id=Delay_2, children [Random_3], properties {delayMax=8, delayMin=3}}
 {id=Delay_3, children [Random_2], properties {delayMax=1.25, delayMin=1}}
 {id=Delay_4, children [Random_4], properties {delayMax=1.5, delayMin=1}}
 {id=Delay_6, children [Random_6], properties {delayMax=6, delayMin=2}}
 {
  id=Enveloper_0
  children [Looping_1]
  properties {
   bLoop=true
   bLoopIndefinitely=true
   loopEnd=60
   pitchCurve {editorCurveData {keys [
      {interpMode=RCIM_Cubic, time=0}
      {arriveTangent=0.0091, interpMode=RCIM_Cubic, leaveTangent=0.0091, tangentMode=RCTM_User, time=30, value=0.8}
      {time=60}
     ]}}
   volumeCurve {editorCurveData {keys [<at 0 1>]}}
  }
 }
 {
  id=Enveloper_4
  children [Enveloper_0]
  properties {
   bLoop=true
   bLoopIndefinitely=true
   loopEnd=20
   pitchCurve {editorCurveData {keys [{time=0}]}}
   volumeCurve {editorCurveData {keys [
      {arriveTangent=-0.01, interpMode=RCIM_Cubic, time=0, value=1}
      <at 10 1.2 RCIM_Cubic>
      {interpMode=RCIM_Cubic, leaveTangent=0.01, time=20, value=1}
     ]}}
  }
 }
 {
  id=Enveloper_6
  children [Mixer_3]
  properties {
   bLoop=true
   bLoopIndefinitely=true
   loopEnd=60
   pitchCurve {editorCurveData {keys [{time=0}]}}
   volumeCurve {editorCurveData {keys [
      {arriveTangent=-0.01, interpMode=RCIM_Cubic, time=0, value=1}
      {arriveTangent=-0.0035, interpMode=RCIM_Cubic, leaveTangent=-0.0035, tangentMode=RCTM_User, time=4.9208, value=0.7335}
      {arriveTangent=0.0016, interpMode=RCIM_Cubic, leaveTangent=0.0016, tangentMode=RCTM_User, time=7.5626, value=0.9613}
      {arriveTangent=0.0008, interpMode=RCIM_Cubic, leaveTangent=0.0008, tangentMode=RCTM_User, time=11.4857, value=0.747}
      {arriveTangent=0.0016, interpMode=RCIM_Cubic, leaveTangent=0.0016, tangentMode=RCTM_User, time=13.3554, value=0.7326}
      {arriveTangent=0.0009, interpMode=RCIM_Cubic, leaveTangent=0.0009, time=15.1401, value=0.8437}
      {arriveTangent=-0.0011, interpMode=RCIM_Cubic, leaveTangent=-0.0011, tangentMode=RCTM_User, time=18.9647, value=0.7376}
      {arriveTangent=0.0002, interpMode=RCIM_Cubic, leaveTangent=0.0002, time=24.105, value=0.9798}
      {arriveTangent=0.0016, interpMode=RCIM_Cubic, leaveTangent=0.0016, tangentMode=RCTM_User, time=28.412, value=0.7398}
      {arriveTangent=-0.0003, interpMode=RCIM_Cubic, leaveTangent=-0.0003, time=31.3775, value=0.8584}
      {arriveTangent=0.0009, interpMode=RCIM_Cubic, leaveTangent=0.0009, tangentMode=RCTM_User, time=35.3006, value=0.7378}
      {arriveTangent=-0.0008, interpMode=RCIM_Cubic, leaveTangent=-0.0008, time=38.0791, value=0.9798}
      <at 42.3861 0.7324>
      {arriveTangent=0.0024, interpMode=RCIM_Cubic, leaveTangent=0.0024, tangentMode=RCTM_User, time=44.9447, value=0.7376}
      {arriveTangent=-0.0006, interpMode=RCIM_Cubic, leaveTangent=-0.0006, time=48.1071, value=0.9613}
      {arriveTangent=-0.0056, interpMode=RCIM_Cubic, leaveTangent=-0.0056, tangentMode=RCTM_User, time=53.5122, value=0.7326}
      <at 55.4914 0.8685>
      {arriveTangent=-0.0003, interpMode=RCIM_Cubic, leaveTangent=-0.0003, tangentMode=RCTM_User, time=58.1315, value=0.7323}
      <at 60 1 RCIM_Cubic>
     ]}}
  }
 }
 {
  id=Enveloper_7
  children [Enveloper_6]
  properties {
   bLoop=true
   bLoopIndefinitely=true
   loopEnd=20
   pitchCurve {editorCurveData {keys [{time=0}]}}
   volumeCurve {editorCurveData {keys [
      {arriveTangent=-0.01, interpMode=RCIM_Cubic, time=0, value=1}
      <at 10 0.8 RCIM_Cubic>
      {interpMode=RCIM_Cubic, leaveTangent=0.01, time=20, value=1}
     ]}}
  }
 }
 {id=Looping_1, children [Random_1]}
 {id=Looping_3, children [Delay_2]}
 {id=Looping_4, children [Delay_3]}
 {id=Looping_5, children [Delay_4]}
 {id=Looping_6, children [Delay_0]}
 {id=Looping_7, children [Delay_6]}
 {id=Mixer_0, children [Modulator_0, Mixer_1, Modulator_2, WavePlayer_46], properties {inputVolume [1,1,1,0.3]}}
 {id=Mixer_1, children [Enveloper_4,Modulator_3], properties {inputVolume [1,1.2]}}
 {id=Mixer_2, children [Attenuation_1, Attenuation_5], properties {inputVolume [1,0.5]}}
 {
  id=Mixer_3
  children [Mixer_6, Modulator_2, WavePlayer_46, Looping_6, Modulator_3, Enveloper_4]
  properties {inputVolume [1,1,0.3,0.25,1,1.25]}
 }
 {id=Mixer_4, children [WavePlayer_40], properties {inputVolume [1.5]}}
 {id=Mixer_6, children [Modulator_1,Modulator_0], properties {inputVolume [1,1]}}
 {id=Modulator_0, children [Looping_5], properties {volumeMax=1, volumeMin=0.7}}
 {id=Modulator_1, children [Looping_4], properties {volumeMax=1, volumeMin=0.7}}
 {id=Modulator_2, children [Looping_3], properties {pitchMin=0.8, volumeMax=1.1, volumeMin=1}}
 {id=Modulator_3, children [Looping_7], properties {pitchMin=0.8, volumeMax=1, volumeMin=0.8}}
 {id=Random_1, children [WavePlayer_9, WavePlayer_10, WavePlayer_12, WavePlayer_13], properties {weights [1,1,1,1]}}
 {id=Random_2, children [WavePlayer_19, WavePlayer_20, WavePlayer_22, WavePlayer_23, WavePlayer_24], properties {weights [1,1,1,1,1]}}
 {
  id=Random_3
  children [WavePlayer_25, WavePlayer_26, WavePlayer_27, WavePlayer_28, WavePlayer_29, WavePlayer_30, WavePlayer_31]
  properties {weights [1,1,1,1,1,1,1]}
 }
 {id=Random_4, children [Mixer_4, WavePlayer_41, WavePlayer_43, WavePlayer_44, WavePlayer_45], properties {weights [1,1,1,1,1]}}
 {
  id=Random_6
  children [
   WavePlayer_51
   WavePlayer_52
   WavePlayer_53
   WavePlayer_54
   WavePlayer_55
   WavePlayer_56
   WavePlayer_57
   WavePlayer_58
   WavePlayer_59
   WavePlayer_60
   WavePlayer_61
   WavePlayer_62
   WavePlayer_63
   WavePlayer_64
   WavePlayer_65
   WavePlayer_66
   WavePlayer_67
   WavePlayer_68
   WavePlayer_69
   WavePlayer_70
   WavePlayer_71
  ]
  properties {weights [1,1,1,1,1,1,1,1,1,1,1,1,1,1,1,1,1,1,1,1,1]}
 }
 {id=Random_7, children [WavePlayer_1, WavePlayer_2, WavePlayer_4, WavePlayer_11], properties {weights [1,1,1,1]}}
 {
  id=WavePlayer_1
  properties {soundWaveAssetPtr=&/Spot_Brimstone_Pillar_Whispers_Detail_Swt_01}
 }
 {
  id=WavePlayer_10
  properties {bLooping=true, soundWaveAssetPtr=&/Spot_Brimstone_Pillar_Tonal_Middle_Loop_02}
 }
 {
  id=WavePlayer_11
  properties {soundWaveAssetPtr=&/Spot_Brimstone_Pillar_Whispers_Detail_Swt_04}
 }
 {
  id=WavePlayer_12
  properties {bLooping=true, soundWaveAssetPtr=&/Spot_Brimstone_Pillar_Tonal_Middle_Loop_03}
 }
 {
  id=WavePlayer_13
  properties {bLooping=true, soundWaveAssetPtr=&/Spot_Brimstone_Pillar_Tonal_Middle_Loop_04}
 }
 {
  id=WavePlayer_19
  properties {soundWaveAssetPtr=&/Spot_Brimestone_Glass_Hand_Impact_06}
 }
 {
  id=WavePlayer_2
  properties {soundWaveAssetPtr=&/Spot_Brimstone_Pillar_Whispers_Detail_Swt_02}
 }
 {
  id=WavePlayer_20
  properties {soundWaveAssetPtr=&/Spot_Brimestone_Glass_Hand_Impact_05}
 }
 {
  id=WavePlayer_22
  properties {soundWaveAssetPtr=&/Spot_Brimestone_Glass_Hand_Impact_03}
 }
 {
  id=WavePlayer_23
  properties {soundWaveAssetPtr=&/Spot_Brimestone_Glass_Hand_Impact_02}
 }
 {
  id=WavePlayer_24
  properties {soundWaveAssetPtr=&/Spot_Brimestone_Glass_Hand_Impact_01}
 }
 {
  id=WavePlayer_25
  properties {soundWaveAssetPtr=&/Spot_Brimestone_Glass_Hand_Rub_01}
 }
 {
  id=WavePlayer_26
  properties {soundWaveAssetPtr=&/Spot_Brimestone_Glass_Hand_Rub_02}
 }
 {
  id=WavePlayer_27
  properties {soundWaveAssetPtr=&/Spot_Brimestone_Glass_Hand_Rub_03}
 }
 {
  id=WavePlayer_28
  properties {soundWaveAssetPtr=&/Spot_Brimestone_Glass_Hand_Rub_04}
 }
 {
  id=WavePlayer_29
  properties {soundWaveAssetPtr=&/Spot_Brimestone_Glass_Hand_Rub_05}
 }
 {
  id=WavePlayer_30
  properties {soundWaveAssetPtr=&/Spot_Brimestone_Glass_Hand_Rub_06}
 }
 {
  id=WavePlayer_31
  properties {soundWaveAssetPtr=&/Spot_Brimestone_Glass_Hand_Rub_07}
 }
 {
  id=WavePlayer_4
  properties {soundWaveAssetPtr=&/Spot_Brimstone_Pillar_Whispers_Detail_Swt_03}
 }
 {
  id=WavePlayer_40
  properties {soundWaveAssetPtr=&Spot_Brimestone_Glass_Hand_Impact_06}
 }
 {
  id=WavePlayer_41
  properties {soundWaveAssetPtr=&Spot_Brimestone_Glass_Hand_Impact_05}
 }
 {
  id=WavePlayer_43
  properties {soundWaveAssetPtr=&Spot_Brimestone_Glass_Hand_Impact_03}
 }
 {
  id=WavePlayer_44
  properties {soundWaveAssetPtr=&Spot_Brimestone_Glass_Hand_Impact_02}
 }
 {
  id=WavePlayer_45
  properties {soundWaveAssetPtr=&Spot_Brimestone_Glass_Hand_Impact_01}
 }
 {
  id=WavePlayer_46
  properties {bLooping=true, soundWaveAssetPtr=&/Spot_Brimstone_Pillar_Proximity_LowEnd_Loop}
 }
 {id=WavePlayer_51, properties {soundWaveAssetPtr=&/Spot_Tormented_Soul_Burst_21}}
 {id=WavePlayer_52, properties {soundWaveAssetPtr=&/Spot_Tormented_Soul_Burst_20}}
 {id=WavePlayer_53, properties {soundWaveAssetPtr=&/Spot_Tormented_Soul_Burst_19}}
 {id=WavePlayer_54, properties {soundWaveAssetPtr=&/Spot_Tormented_Soul_Burst_18}}
 {id=WavePlayer_55, properties {soundWaveAssetPtr=&/Spot_Tormented_Soul_Burst_17}}
 {id=WavePlayer_56, properties {soundWaveAssetPtr=&/Spot_Tormented_Soul_Burst_16}}
 {id=WavePlayer_57, properties {soundWaveAssetPtr=&/Spot_Tormented_Soul_Burst_15}}
 {id=WavePlayer_58, properties {soundWaveAssetPtr=&/Spot_Tormented_Soul_Burst_14}}
 {id=WavePlayer_59, properties {soundWaveAssetPtr=&/Spot_Tormented_Soul_Burst_13}}
 {id=WavePlayer_60, properties {soundWaveAssetPtr=&/Spot_Tormented_Soul_Burst_12}}
 {id=WavePlayer_61, properties {soundWaveAssetPtr=&/Spot_Tormented_Soul_Burst_11}}
 {id=WavePlayer_62, properties {soundWaveAssetPtr=&/Spot_Tormented_Soul_Burst_10}}
 {id=WavePlayer_63, properties {soundWaveAssetPtr=&/Spot_Tormented_Soul_Burst_09}}
 {id=WavePlayer_64, properties {soundWaveAssetPtr=&/Spot_Tormented_Soul_Burst_08}}
 {id=WavePlayer_65, properties {soundWaveAssetPtr=&/Spot_Tormented_Soul_Burst_07}}
 {id=WavePlayer_66, properties {soundWaveAssetPtr=&/Spot_Tormented_Soul_Burst_06}}
 {id=WavePlayer_67, properties {soundWaveAssetPtr=&/Spot_Tormented_Soul_Burst_05}}
 {id=WavePlayer_68, properties {soundWaveAssetPtr=&/Spot_Tormented_Soul_Burst_04}}
 {id=WavePlayer_69, properties {soundWaveAssetPtr=&/Spot_Tormented_Soul_Burst_03}}
 {id=WavePlayer_70, properties {soundWaveAssetPtr=&/Spot_Tormented_Soul_Burst_02}}
 {id=WavePlayer_71, properties {soundWaveAssetPtr=&/Spot_Tormented_Soul_Burst_01}}
 {
  id=WavePlayer_9
  properties {bLooping=true, soundWaveAssetPtr=&/Spot_Brimstone_Pillar_Tonal_Middle_Loop_01}
 }
]
properties {
 bOverrideConcurrency=true
 maxDistance=1300
 soundClassObject=&Fort_Ambient_Props_Athena
 volumeMultiplier=0.501
 attenuationOverrides {
  bEnableFocusInterpolation=true
  bEnableListenerFocus=true
  bEnableSubmixSends=true
  falloffDistance=2000
  focusAttackInterpSpeed=0.1
  focusReleaseInterpSpeed=0.1
  nonFocusVolumeAttenuation=0.5
  nonSpatializedRadiusStart=500
  attenuationShapeExtents {x=925}
  submixSendSettings [{
    bEnableHPFCutoff=false
    bEnableLPFCutoff=false
    disableManualSendClamp=false
    hPFCutoff=20
    lPFCutoff=20000
    maxSendDistance=7500
    maxSendLevel=0.25
    minSendDistance=0
    minSendLevel=0
    sendLevel=0.2
    sendLevelControlMethod=Linear
    soundSubmix=&Fortnite_Reverb_Distant
   }]
 }
 concurrencyOverrides {maxCount=2, voiceStealReleaseTime=2}
}
